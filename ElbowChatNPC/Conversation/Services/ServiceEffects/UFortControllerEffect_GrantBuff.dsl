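; UFortControllerEffect_GrantBuff is a controller effect for NPC conversation services that applies a gameplay effect buff to the requesting player.
; ApplyEffectInternal breaks the effect request context, gets the player's ability system component from it and stores it locally.
; It makes a spec handle from the configurable gameplayEffectToApply class and applies that spec to the player's ability system as a self effect.
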(fn ApplyEffectInternal
  (params (RequestContext :type Effect%20Request%20Context%20Structure%20%28by%20ref%29))
  (bind _output_get (LocalVariables|SetplayerAbilitySystem (Ability|GetAbilitySystemComponent (Utilities|Struct|BreakEffectRequestContext RequestContext))))
  (bind _returnvalue (Spec|MakeSpecHandle(ByClass) (Variables|Default|GetgameplayEffectToApply)))
  (GameplayEffects|ApplyGameplayEffectSpecToSelf _output_get _returnvalue)
  (return))
