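; NPC conversation service effect that grants the requesting player weapons rolled from a configured loot tier group.
; Each loot entry that casts to a weapon definition is added to the inventory, with a random mod set when a scalable-float toggle is on.
; Matching ammo follows, in quantities from a data table that is read into a map on lookup.
; A non-weapon entry ends the loot loop early.
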(fn AddAmmoFromWID
  (params (InventoryOwner :type Fort%20Inventory%20Owner%20Interface%20Interface) (WID :type Fort%20World%20Item%20Definition%20Object%20Reference))
  (bind _returnvalue (WeaponItemDefinition|GetAmmoWorldItemDefinition WID))
  (bind _ammoquantitymap (CallFunction|GetAmmoDatafromTable))
  (if (.ReturnValue (Utilities|Map|Find _ammoquantitymap _returnvalue))
    (Inventory|GiveItemtoInventoryOwner :InventoryOwner InventoryOwner :ItemDefinition _returnvalue :NumberToGive (Utilities|Map|Find _ammoquantitymap _returnvalue) :bNotifyPlayer true :ItemLevel -1 :PickupInstigatorHandle -1 :bUseItemPickupAnalyticEvent false :WeaponAmmoOverride -1)
    (return)
    (else
      (return))))

(fn AddWeaponToPlayerInventory
  (params (weaponItemDef :type Fort%20Weapon%20Item%20Definition%20Object%20Reference) (playerController :type Fort%20Player%20Controller%20Object%20Reference))
  (if (ScalableFloat|GetValueasBool (Variables|Default|Getshould%20Apply%20Weapon%20Mod))
    (bind _returnvalue (Class|FortWeaponModFunctionLibrary|GetRandomWeaponModSetData weaponItemDef))
    (Inventory|GiveWeaponwithModSettoInventoryOwner playerController weaponItemDef _returnvalue)
    (else
      (InventoryManagement|AddItem playerController weaponItemDef 1)))
  (CallFunction|AddAmmofromWID playerController weaponItemDef))

(fn ApplyEffectInternal
  (params (RequestContext :type Effect%20Request%20Context%20Structure%20%28by%20ref%29))
  (bind _playercontroller (CallFunction|GetPlayerReference RequestContext))
  (CallFunction|PickWeaponfromLootPool _playercontroller)
  (return))

(fn GetAmmoDataFromTable
  (bind _outrownames (DataTable|GetDataTableRowNames (Variables|Default|GetammotQuantityDataTable)))
  (for _array_element _outrownames
    (bind _returnvalue (Utilities|GetDataTableRow (Variables|Default|GetammotQuantityDataTable) _array_element))
    (Utilities|Map|Add (Variables|Default|GetammoQuantityMap) (Utilities|Struct|BreakWIDAndIntegerMap _returnvalue) (.amount_5_C4E2E9184959B9E5520959B4EA645336 (Utilities|Struct|BreakWIDAndIntegerMap _returnvalue))))
  (return (Variables|Default|GetammoQuantityMap)))

(fn GetPlayerReference
  (params (EffectRequestContext :type Effect%20Request%20Context%20Structure%20%28by%20ref%29))
  (return (Fortnite|GetFortPlayerControllerfromActor (Utilities|Struct|BreakEffectRequestContext EffectRequestContext))))

(fn PickWeaponFromLootPool
  (params (playerController :type Fort%20Player%20Controller%20Object%20Reference))
  (bind (_outloottodrop _returnvalue) (Loot|PickLootDrops (Variables|Default|Gettier%20Group%20Name) -1 -1 "(GameplayTags=)"))
  (if _returnvalue
    (for _array_element _outloottodrop
      (bind _asfort_weapon_item_definition (Utilities|Casting|CastToFortWeaponItemDefinition (Class|FortItemFunctionLibrary|BreakFortItemEntry _array_element))
        (:then
          (CallFunction|AddWeapontoPlayerInventory _asfort_weapon_item_definition playerController))
        (:CastFailed
          (return false))))
    (return true)
    (else
      (return false))))
